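# Physics asset for the male Viking backpack cosmetic's skeletal mesh, a Battle Royale back accessory.
# A single box body on the backpack_root bone approximates the pack's volume, slightly offset and rotated from the bone.
# It defines no constraints, so it contains no jointed or ragdoll chain.
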PhysicsAsset /BRCosmetics/Accessories/FORT_Backpacks/M_MED_Viking_Backpack/Meshes/SK_Backpack_Male_Viking_PhysicsAsset
bodies {backpack_root [{shapeType=Box, extentX=41.9962, extentY=17.9261, extentZ=41.8048, center=<-0.898 -5.7724 -0.0921>, rotation=<rot 0 3.438 0>}]}
constraints []
